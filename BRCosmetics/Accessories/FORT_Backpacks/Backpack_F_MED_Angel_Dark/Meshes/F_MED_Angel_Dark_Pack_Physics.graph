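# Physics asset for the dark angel backpack cosmetic, defining capsule bodies and joints for its two feathered wings.
# Each wing chains from the backpack root through wing_03 to wing_05, which anchors six feather roots, A to F.
# Feathers A to E add a tip segment, while F is a single segment.
# All joints lock twist and use narrow, asymmetric swing limits, tightest at feather roots, so wings and feathers can only sway slightly.
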PhysicsAsset /BRCosmetics/Accessories/FORT_Backpacks/Backpack_F_MED_Angel_Dark/Meshes/F_MED_Angel_Dark_Pack_Physics
bodies {
 backpack_root [{shapeType=Capsule, length=0, radius=3, shapeName=backpack_root_capsule, rotation=<rot 0 -90 0>}]
 featherA_01_l [{shapeType=Capsule, length=3, radius=2.5, shapeName=featherA_01_l_capsule, rotation=<rot -88.8117 -0.0024 0.0024>}]
 featherA_01_r [{shapeType=Capsule, length=3, radius=2.5, shapeName=featherA_01_r_capsule, rotation=<rot -88.8147 -0.0023 0.0024>}]
 featherA_03_l [{shapeType=Capsule, length=3, radius=1, shapeName=featherA_03_l_capsule, rotation=<rot -88.6949 -0.0013 0.0016>}]
 featherA_03_r [{shapeType=Capsule, length=3, radius=1, shapeName=featherA_03_r_capsule, rotation=<rot -88.6949 -0.0019 0.002>}]
 featherB_01_l [{shapeType=Capsule, length=3, radius=2.5, shapeName=featherB_01_l_capsule, rotation=<rot -88.6688 -0.001 0.001>}]
 featherB_01_r [{shapeType=Capsule, length=3, radius=2.5, shapeName=featherB_01_r_capsule, rotation=<rot -88.674 -0.0017 0.0018>}]
 featherB_03_l [{shapeType=Capsule, length=3, radius=1, shapeName=featherB_03_l_capsule, rotation=<rot -88.8065 -0.002 0.0019>}]
 featherB_03_r [{shapeType=Capsule, length=3, radius=1, shapeName=featherB_03_r_capsule, rotation=<rot -88.8065 -0.0026 0.0027>}]
 featherC_01_l [{shapeType=Capsule, length=3, radius=2.5, shapeName=featherC_01_l_capsule, rotation=<rot -88.674 -0.001 0.0009>}]
 featherC_01_r [{shapeType=Capsule, length=3, radius=2.5, shapeName=featherC_01_r_capsule, rotation=<rot -88.674 -0.0018 0.0017>}]
 featherC_03_l [{shapeType=Capsule, length=3, radius=1, shapeName=featherC_03_l_capsule, rotation=<rot -88.6919 -0.0021 0.002>}]
 featherC_03_r [{shapeType=Capsule, length=3, radius=1, shapeName=featherC_03_r_capsule, rotation=<rot -88.6919 -0.0023 0.0024>}]
 featherD_01_l [{shapeType=Capsule, length=3, radius=2.5, shapeName=featherD_01_l_capsule, rotation=<rot -88.7897 -0.0016 0.0016>}]
 featherD_01_r [{shapeType=Capsule, length=3, radius=2.5, shapeName=featherD_01_r_capsule, rotation=<rot -88.8608 -0.0052 0.0052>}]
 featherD_03_l [{shapeType=Capsule, length=3, radius=1, shapeName=featherD_03_l_capsule, rotation=<rot -88.6946 -0.0017 0.0019>}]
 featherD_03_r [{shapeType=Capsule, length=3, radius=1, shapeName=featherD_03_r_capsule, rotation=<rot -88.6979 -0.002 0.002>}]
 featherE_01_l [{shapeType=Capsule, length=3, radius=2.5, shapeName=featherE_01_l_capsule, rotation=<rot -88.8065 -0.0015 0.0018>}]
 featherE_01_r [{shapeType=Capsule, length=3, radius=2.5, shapeName=featherE_01_r_capsule, rotation=<rot -88.9675 -0.0023 0.0022>}]
 featherE_03_l [{shapeType=Capsule, length=3, radius=1, shapeName=featherE_03_l_capsule, rotation=<rot -88.6919 -0.0024 0.0025>}]
 featherE_03_r [{shapeType=Capsule, length=3, radius=1, shapeName=featherE_03_r_capsule, rotation=<rot -88.6857 -0.0018 0.0017>}]
 featherF_01_l [{shapeType=Capsule, length=3, radius=2.5, shapeName=featherF_01_l_capsule, rotation=<rot -88.6669 -0.0021 0.0024>}]
 featherF_01_r [{shapeType=Capsule, length=3, radius=2.5, shapeName=featherF_01_r_capsule, rotation=<rot -88.6757 -0.0028 0.003>}]
 wing_03_l [{shapeType=Capsule, length=6.9608, radius=6.0724, shapeName=wing_03_l_capsule, rotation=<rot -62.3509 0 -0.0001>}]
 wing_03_r [{shapeType=Capsule, length=6.9002, radius=6.0188, shapeName=wing_03_r_capsule, rotation=<rot -62.3509 0.0001 -0.0001>}]
 wing_05_l [{shapeType=Capsule, length=7.5942, radius=6.3071, shapeName=wing_05_l_capsule}]
 wing_05_r [{shapeType=Capsule, length=7.5942, radius=6.3071, shapeName=wing_05_r_capsule}]
}
constraints [
 {bone1Name=featherA_01_r, bone2Name=wing_05_r, swing1LimitDegrees=5, swing2LimitDegrees=8, twistLimitDegrees=0}
 {bone1Name=wing_03_r, bone2Name=backpack_root, swing1LimitDegrees=5, swing2LimitDegrees=20, twistLimitDegrees=0}
 {bone1Name=wing_05_r, bone2Name=wing_03_r, swing1LimitDegrees=5, swing2LimitDegrees=20, twistLimitDegrees=0}
 {bone1Name=featherF_01_l, bone2Name=wing_05_l, swing1LimitDegrees=5, swing2LimitDegrees=8, twistLimitDegrees=0}
 {bone1Name=featherE_01_l, bone2Name=wing_05_l, swing1LimitDegrees=5, swing2LimitDegrees=8, twistLimitDegrees=0}
 {bone1Name=featherE_03_l, bone2Name=featherE_01_l, swing1LimitDegrees=5, swing2LimitDegrees=12, twistLimitDegrees=0}
 {bone1Name=featherD_01_l, bone2Name=wing_05_l, swing1LimitDegrees=5, swing2LimitDegrees=8, twistLimitDegrees=0}
 {bone1Name=featherD_03_l, bone2Name=featherD_01_l, swing1LimitDegrees=5, swing2LimitDegrees=12, twistLimitDegrees=0}
 {bone1Name=featherC_01_l, bone2Name=wing_05_l, swing1LimitDegrees=5, swing2LimitDegrees=8, twistLimitDegrees=0}
 {bone1Name=featherC_03_l, bone2Name=featherC_01_l, swing1LimitDegrees=5, swing2LimitDegrees=12, twistLimitDegrees=0}
 {bone1Name=wing_03_l, bone2Name=backpack_root, swing1LimitDegrees=5, swing2LimitDegrees=20, twistLimitDegrees=0}
 {bone1Name=wing_05_l, bone2Name=wing_03_l, swing1LimitDegrees=5, swing2LimitDegrees=20, twistLimitDegrees=0}
 {bone1Name=featherA_01_l, bone2Name=wing_05_l, swing1LimitDegrees=5, swing2LimitDegrees=8, twistLimitDegrees=0}
 {bone1Name=featherA_03_l, bone2Name=featherA_01_l, swing1LimitDegrees=5, swing2LimitDegrees=12, twistLimitDegrees=0}
 {bone1Name=featherB_01_l, bone2Name=wing_05_l, swing1LimitDegrees=5, swing2LimitDegrees=8, twistLimitDegrees=0}
 {bone1Name=featherB_03_l, bone2Name=featherB_01_l, swing1LimitDegrees=5, swing2LimitDegrees=12, twistLimitDegrees=0}
 {bone1Name=featherB_01_r, bone2Name=wing_05_r, swing1LimitDegrees=5, swing2LimitDegrees=8, twistLimitDegrees=0}
 {bone1Name=featherB_03_r, bone2Name=featherB_01_r, swing1LimitDegrees=5, swing2LimitDegrees=12, twistLimitDegrees=0}
 {bone1Name=featherA_03_r, bone2Name=featherA_01_r, swing1LimitDegrees=5, swing2LimitDegrees=12, twistLimitDegrees=0}
 {bone1Name=featherC_01_r, bone2Name=wing_05_r, swing1LimitDegrees=5, swing2LimitDegrees=8, twistLimitDegrees=0}
 {bone1Name=featherC_03_r, bone2Name=featherC_01_r, swing1LimitDegrees=5, swing2LimitDegrees=12, twistLimitDegrees=0}
 {bone1Name=featherD_01_r, bone2Name=wing_05_r, swing1LimitDegrees=5, swing2LimitDegrees=8, twistLimitDegrees=0}
 {bone1Name=featherD_03_r, bone2Name=featherD_01_r, swing1LimitDegrees=5, swing2LimitDegrees=12, twistLimitDegrees=0}
 {bone1Name=featherE_01_r, bone2Name=wing_05_r, swing1LimitDegrees=5, swing2LimitDegrees=8, twistLimitDegrees=0}
 {bone1Name=featherE_03_r, bone2Name=featherE_01_r, swing1LimitDegrees=5, swing2LimitDegrees=12, twistLimitDegrees=0}
 {bone1Name=featherF_01_r, bone2Name=wing_05_r, swing1LimitDegrees=5, swing2LimitDegrees=8, twistLimitDegrees=0}
]
